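# EQS query for the ToyFox PilotStandIn's basic rocket, selecting a target actor around the querier.
# A Blueprint generator supplies nearby perception actors as candidates.
# A 2D distance test keeps candidates within a range bound to scalable-float data providers and scores closer ones higher.
# A gameplay tag test filters out any candidate carrying the ToyFox burrowed reposition tag.
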EnvQuery /ToyFoxGameplay/Gameplay/SecondaryPlayer/PilotStandIn/EQS/EQS_ToyFox_PilotStandIn_Rocket_Basic
options [{
  generator {
   name=EnvQueryGenerator_ToyFox_PilotStandIn_NearbyPlayers_C_0
   class=EnvQueryGenerator_ToyFox_PilotStandIn_NearbyPerceptionActors_C
   properties {
    bAutoSortTests=true
    bCanRunAsync=false
    context=&EnvQueryContext_Querier
    envQueryResultNormalizationOption=Default
    generatedItemType=/Script/AIModule.EnvQueryItemType_Actor
    generatorsActionDescription=""
    optionName=""
   }
  }
  tests [
   EnvQueryTest_Distance_0 {
    class=EnvQueryTest_Distance
    properties {
     distanceTo=&EnvQueryContext_Querier
     filterType=Range
     scoringEquation=InverseLinear
     scoringFactor=1
     testComment=""
     testMode=Distance2D
     testPurpose=FilterAndScore
     floatValueMax {
      dataBinding=&self:EnvQueryTest_Distance_0.FortAIDataProvider_ScalableFloat_1
      dataField=FloatValue
      defaultValue=15000
     }
     floatValueMin {
      dataBinding=&self:EnvQueryTest_Distance_0.FortAIDataProvider_ScalableFloat_0
      dataField=FloatValue
      defaultValue=1000
     }
    }
   }
   EnvQueryTest_GameplayTags_0 {
    class=EnvQueryTest_GameplayTags
    properties {
     bDefineReferenceValue=false
     bRejectIncompatibleItems=false
     bSearchComponentsForTags=false
     boolValue=true
     clampMaxType=None
     clampMinType=None
     filterType=Match
     floatValueMax=0
     floatValueMin=0
     multipleContextFilterOp=AllPass
     multipleContextScoreOp=AverageScore
     normalizationType=Absolute
     referenceValue=0
     scoreClampMax=0
     scoreClampMin=0
     scoringEquation=Constant
     scoringFactor=1
     testComment=""
     testPurpose=Filter
     tagQueryToMatch {
      autoDescription=" NONE( Ability.TidalCrane.ToyFox.Reposition.Burrowed )"
      tokenStreamVersion=0
      userDescription=Not burrowed
      queryTokenStream [0,1,3,1,0]
      tagDictionary [{tagName=Ability.TidalCrane.ToyFox.Reposition.Burrowed}]
     }
    }
   }
  ]
 }]
