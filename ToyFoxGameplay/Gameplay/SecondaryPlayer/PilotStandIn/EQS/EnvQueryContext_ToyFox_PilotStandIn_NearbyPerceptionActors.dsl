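; EQS context for pilot stand-ins that supplies the set of nearby perception actors for environment queries.
; It overrides ProvideActorsSet, casts the querier to the Toy Fox giant player pawn, and returns that pawn's result from GetNearbyPerceptionActors with its boolean argument set to true.
; A second sequence branch returns an empty default.
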(fn ProvideActorsSet
  (params (QuerierObject :type Object%20Reference) (QuerierActor :type Actor%20Object%20Reference))
  (Utilities|FlowControl|Sequence
    (:then_0
      (bind _asb_toy_fox_giant_player_pawn (Utilities|Casting|CastToB_ToyFoxGiantPlayerPawn QuerierActor))
      (bind _actors (StandinPilots|GetNearbyPerceptionActors _asb_toy_fox_giant_player_pawn true))
      (return _actors))
    (:then_1
      (return 0))))
