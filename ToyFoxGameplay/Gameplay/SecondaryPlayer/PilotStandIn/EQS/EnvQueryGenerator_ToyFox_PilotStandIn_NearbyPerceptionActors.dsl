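; Blueprint EQS generator that supplies nearby perception actors as query items for the ToyFox pilot stand-in.
; It casts the querier to the giant player pawn and gets that pawn's nearby perception actors through the StandinPilots function.
; Each returned actor is added as a generated item; the context actors input is unused.
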(fn DoItemGenerationFromActors
  (params (ContextActors :type Array%20of%20Actor%20Object%20References))
  (Utilities|FlowControl|Sequence
    (:then_0
      (bind _asb_toy_fox_giant_player_pawn (Utilities|Casting|CastToB_ToyFoxGiantPlayerPawn (EQS|GetQuerier)))
      (bind _actors (StandinPilots|GetNearbyPerceptionActors _asb_toy_fox_giant_player_pawn true))
      (for _array_element _actors
        (EQS|AddGeneratedActor _array_element)))
    (:then_1
      (return))))
